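# Editor utility whose Update Art function swaps selected level actors for refreshed art assets defined in DT_EnvActorRefresh.
# It resolves each actor's content asset path, then compares it to every row's original asset path using a wildcard string match.
# On a match, it selects that actor alone and calls ReplaceSelectedActors with the row's refreshed asset path.
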EditorUtilityBlueprint /Game/NeverCook/EditorUtilities/Environment/ActorSwapUtility/AAU_EnvActorSwap
graphs [
 EventGraph {links [], nodes []}
 Update Art {nodes [
   Update Art() {is=title K2Node_FunctionEntry}
   > n1 = GetSelectedActors()
   > n2 = For Each Loop(Array:Array of Actor Object References=@n1) -> (Array Element:Actor Object Reference)
   > n3 = Set ActorToReplace(ActorToReplace:Actor Object Reference=@n2) -> (Output_Get:Actor Object Reference)
   > n4 = Get Content Asset Path From Actor Ref!(execute, Actor:Object Reference=@n3, __WorldContext:Object Reference=Self-Reference()) -> (then, Content Asset Data:Asset Data Structure, Valid:bool) {is=title K2Node_CallFunction}
   > Branch(Condition=@n4.Valid)
   > n5 = ToSoftObjectPath!(execute, InAssetData:"Asset Data Structure (by ref)"=@n4) -> (then, ReturnValue:Soft Object Path Structure)
   > Set AssetToCheck(AssetToCheck:Soft Object Path Structure=@n5) -> (Output_Get:Soft Object Path Structure)
   > n6 = GetDataTableRowNames(Table=Get DT_EnvActorRefresh() -> (DT_EnvActorRefresh:Data Table Object Reference))
   > n7 = For Each Loop(Array:Array of Names=@n6) -> (Array Element:Name)
   > n8 = Sequence()
   > n9 = Get Data Table Row!(execute:Exec, DataTable:Data Table Object Reference=Get DT_EnvActorRefresh(), RowName:Name=@n7) -> (then:Exec, RowNotFound:Exec, ReturnValue:STR Env Actor Refresh Structure)
   n10 = Break STR Env Actor Refresh!(STR_EnvActorRefresh:"STR Env Actor Refresh Structure (by ref)"=@n9) -> (
    OriginalAsset_2_#1:Soft Object Path Structure
    RefreshedAsset_4_#2:Soft Object Path Structure
   ) {is=title K2Node_BreakStruct}
   > n11 = Set OldArtAsset!(execute, OldArtAsset:Soft Object Path Structure=@n10) -> (then, Output_Get_PathString:String)
   > n12 = Set FoundEntry(FoundEntry:bool=MatchesWildcard(SourceString=@n11, Wildcard=Get!() -> (AssetToCheck_PathString:String))) -> (Output_Get:bool)
   > Branch(Condition=@n12)
   > Set ReplacementAsset(ReplacementAsset:Soft Object Path Structure=@n10.RefreshedAsset_4_#2) -> (Output_Get:Soft Object Path Structure)
   Branch(execute=@n8.then_1, Condition=Get FoundEntry() -> (FoundEntry:bool))
   > SetSelectedLevelActors(
    self=Get EditorActorSubsystem()
    ActorsToSelect=Make Array("[0]":Actor Object Reference=Get ActorToReplace() -> (ActorToReplace:Actor Object Reference)) -> (Array:Array of Actor Object References)
   )
   > ReplaceSelectedActors!(execute, InAssetPath:String=Get!() -> (ReplacementAsset_PathString:String)) -> (then)
   Set OldArtAsset!(execute, OldArtAsset) -> (then, Output_Get:Soft Object Path Structure)
   > Set ReplacementAsset()
  ]}
]
variables [
 ActorToReplace {type=Actor, default=null}
 AssetToCheck {type=SoftObjectPath, default=""}
 DT_EnvActorRefresh {type=DataTable, default=/Game/NeverCook/EditorUtilities/Environment/ActorSwapUtility/Data/DT_EnvActorRefresh}
 FoundEntry {type=boolean}
 OldArtAsset {type=SoftObjectPath, default=""}
 ReplacementAsset {type=SoftObjectPath, default=""}
]
