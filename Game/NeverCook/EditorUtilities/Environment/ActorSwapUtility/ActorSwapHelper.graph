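# ActorSwapHelper is an editor utility widget that swaps legacy environment art on level actors for refreshed assets.
# Update Visuals stores ReplaceLegacyArt's map of assets to actor sets, which Tick drains one entry at a time by selecting the actors and calling ReplaceSelectedActors.
# A details view exposes the widget's own Replacement Modifiers category.
# A ">>" button, visible only to one hardcoded username, uses Python to open the widget's Blueprint.
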EditorUtilityWidgetBlueprint /Game/NeverCook/EditorUtilities/Environment/ActorSwapUtility/ActorSwapHelper
graphs [EventGraph {nodes [
   "On Clicked (BTN_EditWidget)"!() -> (OutputDelegate, then, widget:User Widget Object Reference, event:Pointer Event Structure) {is=title K2Node_ComponentBoundEvent}
   > n1 = GetBlueprintAsset(Object=&self)
   > Execute Python Script!(
    execute
    PythonScript="import unreal\r\n\r\nmy_widget = widget_object\r\nasset_editor_subsystem = unreal.get_editor_subsystem(unreal.AssetEditorSubsystem)\r\nasset_editor_subsystem.open_editor_for_assets([my_widget])"
    widget_object:Blueprint Object Reference=@n1
   ) -> (then, ReturnValue) {inputs [widget_object]}
   n2 = "On Clicked (BTN_SwapAssets)"!() -> (OutputDelegate, then, widget:User Widget Object Reference, event:Pointer Event Structure) {is=title K2Node_ComponentBoundEvent}
   > n3 = ReplaceLegacyArt!(execute, __WorldContext:Object Reference=@n2.widget) -> (then, L_RefreshedActor_Map1:Map of Soft Object References to STR Refreshed Actor Array Structures)
   > Set L Refreshed Actor Map 1(L Refreshed Actor Map 1:Map of Soft Object References to STR Refreshed Actor Array Structures=@n3) -> (Output_Get:Map of Soft Object References to STR Refreshed Actor Array Structures)
   Event PreConstruct()
   > SetObject(self=Get DV_ReplacementModifiers() -> (DV_ReplacementModifiers:Details View Object Reference), NewObject=Self-Reference())
   Event Construct()
   > n4 = Branch(Condition="Equal Exactly (String)"(A=GetPlatformUserName!() -> (ReturnValue:String), B=brianfasten))
   n5 = Get BTN_EditWidget() -> (BTN_EditWidget:TA Generic 01 Button Object Reference)
   > SetVisibility(self=@n5, InVisibility=Visible)
   SetVisibility(execute=@n4.else, self=@n5, InVisibility=Hidden)
   Event Tick()
   n6 = Get L Refreshed Actor Map 1() -> (L Refreshed Actor Map 1:Map of Soft Object References to STR Refreshed Actor Array Structures)
   > Branch(Condition="integer > integer"(A=Length!(TargetMap:Map of Soft Object References to STR Refreshed Actor Array Structures=@n6) -> (ReturnValue:int=0) {is=title K2Node_CallFunction}))
   n7 = Get L Refreshed Actor Map 1()
   > n8 = Keys!(execute, TargetMap:Map of Soft Object References to STR Refreshed Actor Array Structures=@n7) -> (then, Keys:Array of Soft Object References)
   > n9 = Values!(execute, TargetMap:Map of Soft Object References to STR Refreshed Actor Array Structures=@n7) -> (then, Values:Array of STR Refreshed Actor Array Structures)
   > n10 = To Array!(
    execute
    A:Set of Actor Object References=Break STR Refreshed Actor Array!(STR_RefreshedActorArray:"STR Refreshed Actor Array Structure (by ref)"="Get (a copy)"(Array:Array of STR Refreshed Actor Array Structures=@n9) -> (Output:STR Refreshed Actor Array Structure)) -> (Actors_4_#1:Set of Actor Object References) {is=title K2Node_BreakStruct}
   ) -> (then, Result:Array of Actor Object References)
   > PrintString(InString=setting selected)
   > SetSelectedLevelActors(self=Get EditorActorSubsystem(), ActorsToSelect=@n10)
   n11 = "Get (a copy)"(Array:Array of Soft Object References=@n8) -> (Output:Soft Object Reference)
   > ReplaceSelectedActors!(
    execute
    InAssetPath:String="To String (SoftObjectReference)"!(SoftObjectReference:"Soft Object Reference (by ref)"=@n11) -> (ReturnValue:String)
   ) -> (then)
   > Remove!(
    execute
    TargetMap:Map of Soft Object References to STR Refreshed Actor Array Structures=Get L Refreshed Actor Map 1()
    Key:"Soft Object Reference (by ref)"=@n11
   ) -> (then, ReturnValue:bool)
  ]}]
variables [
 KeepDataOverrides {type=boolean}
 L Refreshed Actor Map 1 {default {}}
 Refreshed Asset {type=Object, default=null}
]
widgets {
 info {rootWidgetClass=&HorizontalBox, widgetCount=5}
 widgets [
  HorizontalBox_0 {widgetClassPath=&HorizontalBox, widget=HorizontalBox_0}
  VerticalBox_0 {widgetClassPath=&VerticalBox, parent=HorizontalBox_0, widget=VerticalBox_0, slot {class=HorizontalBoxSlot}}
  DV_ReplacementModifiers {
   widgetClassPath=&DetailsView
   parent=VerticalBox_0
   widget=DV_ReplacementModifiers
   properties {categoriesToShow [Replacement Modifiers]}
   slot {class=VerticalBoxSlot}
  }
  BTN_SwapAssets {
   widgetClassPath=&TA_Generic01_Button
   parent=VerticalBox_0
   widget=BTN_SwapAssets
   properties {label=Update Visuals}
   slot {class=VerticalBoxSlot}
  }
  BTN_EditWidget {
   widgetClassPath=&TA_Generic01_Button
   parent=HorizontalBox_0
   widget=BTN_EditWidget
   properties {label=">>", buttonColor_Default=<rgba 0.0469 0.0469 0.0469 1>, sizeOverride=<20 0>}
   slot {class=HorizontalBoxSlot, properties {padding=<ltrb 5 5 5 5>}}
  }
 ]
}
