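# EUC_ArtRefresh is an editor utility Blueprint in the environment ActorSwapUtility folder, apparently a stub for refreshing art assets.
# It declares a single object variable, RefreshedAsset, which defaults to null.
# The event graph holds only unconnected BeginPlay and Tick events, so it has no logic of its own.
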EditorUtilityBlueprint /Game/NeverCook/EditorUtilities/Environment/ActorSwapUtility/EUC_ArtRefresh
graphs [EventGraph {nodes [
   Event Begin Play()
   Event Tick!() -> (OutputDelegate, then, DeltaSeconds:float)
  ]}]
variables [RefreshedAsset {type=Object, default=null}]
