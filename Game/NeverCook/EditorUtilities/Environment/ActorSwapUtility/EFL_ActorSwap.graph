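# Editor utility Blueprint for swapping legacy environment art in a level for refreshed assets, via a ReplaceLegacyArt function.
# It scans all level actors for EUC_ArtRefresh components, groups those with valid RefreshedAsset references into a map of asset to actor-set struct, and returns it.
# Separate nodes for selecting grouped actors (always the first group) and calling ReplaceSelectedActors are unwired, so the function only gathers.
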EditorUtilityBlueprint /Game/NeverCook/EditorUtilities/Environment/ActorSwapUtility/EFL_ActorSwap
graphs [ReplaceLegacyArt {nodes [
   n1 = Get L_RefreshedActor_Map() -> (L_RefreshedActor_Map:Map of Soft Object References to STR Refreshed Actor Array Structures)
   n2 = Values!(execute, TargetMap:Map of Soft Object References to STR Refreshed Actor Array Structures=@n1) -> (then, Values:Array of STR Refreshed Actor Array Structures)
   > n3 = Keys!(execute, TargetMap:Map of Soft Object References to STR Refreshed Actor Array Structures=@n1) -> (then, Keys:Array of Soft Object References)
   > n4 = For Each Loop(Array:Array of Soft Object References=@n3) -> (Array Element:Soft Object Reference)
   > Set L_RefreshedAsset(L_RefreshedAsset:Soft Object Reference=@n4) -> (Output_Get:Soft Object Reference)
   > n5 = To Array!(
    execute
    A:Set of Actor Object References=Break STR Refreshed Actor Array!(STR_RefreshedActorArray:"STR Refreshed Actor Array Structure (by ref)"="Get (a ref)"(Array:Array of STR Refreshed Actor Array Structures=@n2, Dimension 1=0) -> (Output:"STR Refreshed Actor Array Structure (by ref)")) -> (Actors_4_#1:Set of Actor Object References) {is=title K2Node_BreakStruct}
   ) -> (then, Result:Array of Actor Object References)
   > PrintString(InString=setting selected)
   > SetSelectedLevelActors(self=Get EditorActorSubsystem(), ActorsToSelect=@n5)
   ReplaceLegacyArt() {is=title K2Node_FunctionEntry}
   > n6 = GetAllLevelActors(self=Get EditorActorSubsystem())
   > n7 = For Each Loop(Array:Array of Actor Object References=@n6) -> (Array Element:Actor Object Reference)
   Return Node!(
    execute=@n7.Completed
    L_RefreshedActor_Map1:Map of Soft Object References to STR Refreshed Actor Array Structures=Get L_RefreshedActor_Map()
   )
   > n8 = Set L_Actor(L_Actor:Actor Object Reference=@n7) -> (Output_Get:Actor Object Reference)
   n9 = GetComponentByClass!(
    self:Actor Object Reference=@n8
    ComponentClass:Actor Component Class Reference=&/EUC_ArtRefresh.EUC_ArtRefresh_C
   ) -> (ReturnValue:EUC Art Refresh Object Reference)
   > Branch(Condition=IsValid(Object=@n9))
   n10 = Get RefreshedAsset!(self:EUC Art Refresh Object Reference=@n9) -> (RefreshedAsset:Soft Object Reference)
   > Set L_RefreshedAsset(L_RefreshedAsset=@n10)
   > Branch(Condition=IsValidSoftObjectReference(SoftObjectReference=@n10))
   n11 = Find(
    TargetMap:Map of Soft Object References to STR Refreshed Actor Array Structures=Get L_RefreshedActor_Map()
    Key:"Soft Object Reference (by ref)"=@n10
   ) -> (Value:STR Refreshed Actor Array Structure)
   > n12 = Branch(Condition=@n11.ReturnValue)
   > n13 = Set L_STR(L_STR:STR Refreshed Actor Array Structure=@n11) -> (Output_Get:STR Refreshed Actor Array Structure)
   > n14 = Set L_SET(L_SET:Set of Actor Object References=Break STR Refreshed Actor Array(STR_RefreshedActorArray=@n13)) -> (Output_Get:Set of Actor Object References)
   > Add!(
    execute
    TargetSet:Set of Actor Object References=@n14
    NewItem:"Actor Object Reference (by ref)"=Get L_Actor() -> (L_Actor:Actor Object Reference)
   ) -> (then) {is=title K2Node_CallFunction}
   > n15 = Set members in STR Refreshed Actor Array!(
    execute:Exec
    StructRef:"STR Refreshed Actor Array Structure (by ref)"=Get L_STR() -> (L_STR:STR Refreshed Actor Array Structure)
    Actors_4_#1:Set of Actor Object References=Get L_SET() -> (L_SET:Set of Actor Object References)
   ) -> (then:Exec, StructOut:"STR Refreshed Actor Array Structure (by ref)") {is=title K2Node_SetFieldsInStruct}
   > Add!(
    execute
    TargetMap:Map of Soft Object References to STR Refreshed Actor Array Structures=Get L_RefreshedActor_Map()
    Key:"Soft Object Reference (by ref)"=Get L_RefreshedAsset() -> (L_RefreshedAsset:Soft Object Reference)
    Value:"STR Refreshed Actor Array Structure (by ref)"=@n15
   ) -> (then)
   n16 = Get L_Actor()
   Add(
    execute=@n12.else
    TargetMap=Get L_RefreshedActor_Map()
    Key=Get L_RefreshedAsset()
    Value=Make STR Refreshed Actor Array!(Actors_4_#1:Set of Actor Object References=Make Set!("[0]":Actor Object Reference=@n16) -> (Set:Set of Actor Object References) {is=title K2Node_MakeSet}) -> (STR_RefreshedActorArray:STR Refreshed Actor Array Structure) {is=title K2Node_MakeStruct}
   )
   ReplaceSelectedActors!(
    execute
    InAssetPath:String="To String (SoftObjectReference)"!(SoftObjectReference:"Soft Object Reference (by ref)"=Get L_RefreshedAsset()) -> (ReturnValue:String)
   ) -> (then)
  ]}]
variables []
